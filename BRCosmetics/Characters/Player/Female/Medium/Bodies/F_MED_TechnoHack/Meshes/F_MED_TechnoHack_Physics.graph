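# Physics asset for the TechnoHack female medium player cosmetic body, covering a bag accessory and the left thigh rather than the whole body.
# It holds three capsule bodies: a tiny one on bag_auto, a small one on bag_offset and a long one on the left thigh twist bone.
# A single constraint with zeroed swing and twist limit angles links bag_offset to bag_auto, leaving the thigh capsule unconstrained.
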PhysicsAsset /BRCosmetics/Characters/Player/Female/Medium/Bodies/F_MED_TechnoHack/Meshes/F_MED_TechnoHack_Physics
bodies {
 bag_auto [{shapeType=Capsule, length=0.505, radius=0.505, shapeName=bag_auto_capsule, rotation=<rot 0 -90 0>}]
 bag_offset [{
   shapeType=Capsule
   length=7.5189
   radius=2.1735
   shapeName=bag_offset_capsule
   center=<2.4832 -0.1518 -4.2637>
   rotation=<rot 0 178.38 -90.6302>
  }]
 thigh_twist_01_l [{shapeType=Capsule, length=52.2304, radius=7.6223, shapeName=thigh_twist_01_l_capsule, center=<0 -0.3951 0.8573>, rotation=<rot 90 -4.8462 0>}]
}
constraints [{bone1Name=bag_offset, bone2Name=bag_auto, swing1LimitDegrees=0, swing2LimitDegrees=0, twistLimitDegrees=0}]
