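# Animation Blueprint for the female medium TechnoHack body cosmetic, following the attached parent mesh's pose and adding secondary motion.
# It layers facial root, jaw and chin bones copied from the player's head mesh, cached when character customization loads, over the parent pose.
# The blended pose runs through the TechnoHack Control Rig, then a RigidBody simulation in pelvis bone space using the outfit's physics asset.
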AnimBlueprint /BRCosmetics/Characters/Player/Female/Medium/Bodies/F_MED_TechnoHack/Meshes/F_MED_TechnoHack_AnimBP
target_skeleton=&/F_MED_TechnoHack_Skeleton
graphs [
 AnimGraph {nodes [
   n1 = Control Rig(Source=Use cached pose 'Start_Cache'()) {
    alphaBoolBlend {blendOption=Linear}
    controlRigAssetReference {blueprintRigClass=&/F_MED_TechnoHack_CtrlRig.F_MED_TechnoHack_CtrlRig_C}
   }
   n2 = RigidBody(ComponentPose=Local To Component(LocalPose=@n1)) {
    bClampLinearTranslationLimitToRefPose=true
    overridePhysicsAsset=&/F_MED_TechnoHack_Physics
    simulationSpace=BaseBoneSpace
    alphaBoolBlend {blendOption=Linear}
    baseBoneRef {boneName=pelvis}
    simSpaceSettings {dampingAlpha=0, maxAngularAcceleration=0, maxAngularVelocity=0, maxLinearAcceleration=0, maxLinearVelocity=0, velocityScaleZ=0}
   }
   Output Pose(Result=Component To Local(ComponentPose=@n2)) {layerGroup=None}
   n3 = Layered blend per bone(
    BasePose=Copy Pose From Mesh() {bUseAttachedParent=true}
    BlendPoses_0=Copy Pose From Mesh(SourceMeshComponent="Get Face Mesh "() -> ("Face Mesh ":Skeletal Mesh Component Object Reference)) {bUseAttachedParent=true}
   ) {layerSetup [{branchFilters [
       {boneName=FACIAL_C_FacialRoot}
       {boneName=FACIAL_C_Jaw}
       {boneName=FACIAL_L_UnderChin}
       {boneName=FACIAL_R_UnderChin}
       {boneName=FACIAL_L_Jawline}
       {boneName=FACIAL_R_Jawline}
       {boneName=FACIAL_C_Jawline}
       {boneName=FACIAL_L_ChinSide}
       {boneName=FACIAL_R_ChinSide}
       {boneName=FACIAL_C_Chin}
      ]}]}
   Save cached pose 'Start_Cache'(Pose=@n3) {cacheName=Start_Cache}
  ]}
 EventGraph {nodes [
   Event OnCharacterCustomizationLoaded()
   > n1 = Cast To PlayerPawn_Athena(Object=TryGetPawnOwner())
   > "Set Face Mesh "("Face Mesh ":Skeletal Mesh Component Object Reference=GetSkeletalMeshForPartType(self=@n1, PartType=Head)) -> (Output_Get:Skeletal Mesh Component Object Reference)
   Event BlueprintUpdateAnimation()
  ]}
]
variables ["Face Mesh " {type=SkeletalMeshComponent}]
