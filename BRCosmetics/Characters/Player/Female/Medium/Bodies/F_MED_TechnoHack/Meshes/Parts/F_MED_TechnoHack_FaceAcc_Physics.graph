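# Physics asset for the TechnoHack female medium character's FaceAcc part, setting up a simulated long-hair bone chain.
# Five small capsules form the dyn_long_hair chain, hung from a base hair body that links to the head with free rotation.
# Swing limits generally widen from root to tip, with zero twist, so the tip bends more than the root.
# Larger capsules cover the head, left clavicle and left upper arm only, apparently as colliders for the hair.
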PhysicsAsset /BRCosmetics/Characters/Player/Female/Medium/Bodies/F_MED_TechnoHack/Meshes/Parts/F_MED_TechnoHack_FaceAcc_Physics
bodies {
 base_long_hair_0 [{shapeType=Capsule, length=0.505, radius=0.505, shapeName=base_long_hair_0_capsule, rotation=<rot 0 -90 0>}]
 clavicle_l [{
   shapeType=Capsule
   length=7.7057
   radius=6.9774
   shapeName=clavicle_l_capsule
   center=<2.7618 -0.1953 -0.7929>
   rotation=<rot 82.4642 -0.0009 -0.001>
  }]
 dyn_long_hair_0 [{
   shapeType=Capsule
   length=3.5506
   radius=1.405
   shapeName=dyn_long_hair_0_capsule
   center=<2 0 0>
   rotation=<rot 85.8623 -179.9996 -179.9997>
  }]
 dyn_long_hair_1 [{
   shapeType=Capsule
   length=3.4876
   radius=1.468
   shapeName=dyn_long_hair_1_capsule
   center=<1.984 0 0.6399>
   rotation=<rot 80.9205 -179.9998 179.9999>
  }]
 dyn_long_hair_2 [{
   shapeType=Capsule
   length=3.5914
   radius=1.3643
   shapeName=dyn_long_hair_2_capsule
   center=<2.0083 0 0.6941>
   rotation=<rot 87.5635 0.0001 0.0001>
  }]
 dyn_long_hair_3 [{shapeType=Capsule, length=3.3711, radius=1.5845, shapeName=dyn_long_hair_3_capsule, center=<2 0 0>, rotation=<rot 90 0 0>}]
 dyn_long_hair_4 [{shapeType=Capsule, length=3.3711, radius=1.5845, shapeName=dyn_long_hair_4_capsule, center=<2 0 0>, rotation=<rot 90 0 0>}]
 head [{shapeType=Capsule, length=11.682, radius=7.3822, shapeName=head_capsule, center=<2.9938 1.2575 -0.5956>, rotation=<rot 62.7395 -34.919 180>}]
 upperarm_l [{shapeType=Capsule, length=24.5189, radius=4.5978, shapeName=upperarm_l_capsule, center=<9.8191 -0.4461 -0.6444>, rotation=<rot 90 -3.0154 0>}]
}
constraints [
 {
  bone1Name=upperarm_l
  bone2Name=clavicle_l
  swing1LimitDegrees=45
  swing1Motion=Free
  swing2LimitDegrees=45
  swing2Motion=Free
  twistLimitDegrees=45
  twistMotion=Free
 }
 {
  bone1Name=base_long_hair_0
  bone2Name=head
  swing1LimitDegrees=45
  swing1Motion=Free
  swing2LimitDegrees=45
  swing2Motion=Free
  twistLimitDegrees=45
  twistMotion=Free
 }
 {bone1Name=dyn_long_hair_0, bone2Name=base_long_hair_0, swing1LimitDegrees=0, swing2LimitDegrees=15, twistLimitDegrees=0}
 {bone1Name=dyn_long_hair_1, bone2Name=dyn_long_hair_0, swing1LimitDegrees=5, swing2LimitDegrees=20, twistLimitDegrees=0}
 {bone1Name=dyn_long_hair_2, bone2Name=dyn_long_hair_1, swing1LimitDegrees=10, swing2LimitDegrees=25, twistLimitDegrees=0}
 {bone1Name=dyn_long_hair_3, bone2Name=dyn_long_hair_2, swing1LimitDegrees=10, swing2LimitDegrees=30, twistLimitDegrees=0}
 {bone1Name=dyn_long_hair_4, bone2Name=dyn_long_hair_3, swing1LimitDegrees=10, swing2LimitDegrees=35, twistLimitDegrees=0}
]
